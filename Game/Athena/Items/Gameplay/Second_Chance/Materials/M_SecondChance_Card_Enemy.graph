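# Translucent unlit material for the enemy variant of the Second Chance card pickup, drawn as a hologram with a timed appearance animation.
# StartTime and Duration drive world position offset that tilts the card and settles it from a narrowed, vertically displaced shape.
# Emissive blends a Fresnel tint, using normals rotated to match, with a pulsing masked glow that Low quality simplifies.
# Opacity combines Fresnel, screen-space scanlines faded by camera distance, and a texture-mask minimum.
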Material /Game/Athena/Items/Gameplay/Second_Chance/Materials/M_SecondChance_Card_Enemy
graph {
 properties {
  blendMode=BLEND_Translucent
  pixelDepthOffsetMode=PDOM_Legacy
  refractionMethod=RM_IndexOfRefraction
  shadingModel=MSM_Unlit
  usedShadingModels=Unlit
 }
 nodes [
  n1 = Normalize(VectorInput=VectorParameter() {parameterName=RestingRotation, defaultValue=<rgba 0.3 0.2 0 1>}.RGB)
  n2 = Time()
  n3 = ScalarParameter() {parameterName=StartTime}
  n4 = Clamp(None=Divide(A=Subtract(A=@n2, B=@n3), B=ScalarParameter() {defaultValue=1, parameterName=Duration}))
  n5 = Multiply(A=ScalarParameter() {defaultValue=0.08, parameterName=WPORotationTilt}, B=Power(Base=@n4))
  n6 = MaterialFunctionCall() {materialFunction=&ObjectPivotPoint}
  n7 = WorldPosition()
  n8 = Constant3Vector() {constant=<rgba 0 0 -0.25 0>}
  n9 = ScalarParameter() {defaultValue=72}
  n10 = Subtract(A=WorldPosition().XYZ, B=Add(A=ObjectPositionWS(), B=Multiply(A=@n8, B=@n9) {constA=-0.25}))
  n11 = LinearInterpolate(Alpha=@n4) {constA=1, constB=0}
  n12 = Power(Base=@n11) {constExponent=5}
  n13 = Multiply(A=ComponentMask(None=@n10) {g=true, r=true}, B=@n12)
  n14 = Power(Base=@n11)
  n15 = Divide(A=ComponentMask(None=@n10) {b=true}, B=@n9)
  n16 = LinearInterpolate(Alpha=@n4) {constA=1, constB=3}
  n17 = Sine(None=Add(A=Multiply(A=@n4, B=@n16)) {constB=0.25}) {period=0.75}
  n18 = Multiply(A=@n14, B=Multiply(A=Clamp(None=@n15) {minDefault=0.3}, B=Multiply(A=@n14, B=@n17)))
  Add(
   A=Multiply(A=RotateAboutAxis(NormalizedRotationAxis=@n1, RotationAngle=@n5, PivotPoint=@n6.Object Pivot Location, Position=@n7.XYZ), B=@n4)
   B=Multiply(A=VectorParameter() {parameterName=WPOMax, defaultValue=<rgba -0.75 -0.75 250 1>}.RGB, B=AppendVector(A=@n13, B=@n18)) {constB=0}
  )
  n19 = TextureSampleParameter2D() {
   desc="R = SpecMask G = Metallic B = Roughness"
   group=Specular
   parameterName=SpecularMasks
   samplerSource=SSM_Wrap_WorldGroupSettings
   samplerType=SAMPLERTYPE_Masks
   texture=/Game/Athena/Items/Gameplay/Second_Chance/Textures/SecondChance_Card_Pickup_S
  }
  n20 = Multiply(A=MaterialFunctionCall(In=@n19.B, Contrast=ScalarParameter() {defaultValue=2, parameterName=Contrast}) {materialFunction=&CheapContrast}.Result) {constB=0.1}
  n21 = Add(A=@n20) {constB=0.1}
  n22 = TextureSample() {samplerType=SAMPLERTYPE_Normal, texture=/Game/Athena/Items/Gameplay/Second_Chance/Textures/SecondChance_Card_Pickup_N}
  n23 = Fresnel(Normal=MaterialFunctionCall(Rotation Axis=@n1, Rotation Angle=@n5, World Space Vertex Normals=Transform(None=@n22.RGB)) {materialFunction=&FixRotateAboutAxisNormals}.New World Space Vertex Normal) {baseReflectFraction=0.02, exponent=3}
  n24 = TextureSample() {texture=/Game/Athena/Items/Gameplay/Second_Chance/Textures/SecondChance_Card_Pickup_E}
  n25 = Add(A=Clamp(None=LinearInterpolate(Alpha=@n11) {constA=1, constB=-2.5})) {constB=0.25}
  n26 = OneMinus(None=ConstantBiasScale(None=Sine(None=@n25) {period=1.25}))
  n27 = Fresnel(ExponentIn=Multiply(A=@n4) {constB=2}) {baseReflectFraction=0.1}
  n28 = Add(A=Multiply(A=@n24.B, B=Multiply(A=@n26) {constA=1, constB=5}), B=Multiply(A=Multiply(A=@n27, B=@n12)) {constB=8})
  Add(
   A=Multiply(A=LinearInterpolate(A=@n21, Alpha=@n23) {constA=0.1}, B=VectorParameter() {parameterName=Color, defaultValue=<rgba 0.6406 0.9346 1.5 0>}.RGB)
   B=Multiply(A=VectorParameter() {parameterName=GlowColor, defaultValue=<rgba 2 0.5665 0.1855 0>}.RGB, B=QualitySwitch(Default=@n28, Low=@n24.B))
  )
  n29 = Multiply(A=@n23) {constB=2}
  n30 = ScalarParameter() {defaultValue=0.4, parameterName=ScanlineAdd}
  n31 = Multiply(A=ComponentMask(None=ScreenPosition().ViewportUV) {g=true}, B=ScalarParameter() {defaultValue=150, parameterName=ScanlineTile})
  n32 = Power(Base=ConstantBiasScale(None=Sine(None=@n31)))
  n33 = MaterialFunctionCall(Distance=ScalarParameter() {defaultValue=1000, parameterName=FadeDistance}) {materialFunction=&Near_Camera_Fade}
  Clamp(
   None=Add(A=Add(A=@n29, B=@n30), B=LinearInterpolate(A=@n32, Alpha=@n33.Result) {constA=0.4, constB=0.3}) {constB=0.2}
   Min=Add(A=@n20, B=@n24.B)
  )
  Constant() {r=1}
  Multiply(A=Time()) {constB=0.1}
  Time() {bOverride_Period=true, period=1}
 ]
}
