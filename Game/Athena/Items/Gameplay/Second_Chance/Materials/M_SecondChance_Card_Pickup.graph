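# Lit material for the Second Chance card pickup, with textured surface maps, a timed emissive glow and vertex animation.
# StartTime and Duration drive a clamped timeline that pulses the emissive map, fades a Fresnel glow, tilts the card about RestingRotation with corrected normals and adds a WPOMax-scaled sine deformation.
# The MeshParticle switch instead uses particle dynamic parameters to drive the deformation and a height-based Fresnel wipe, which is zeroed on mobile.
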Material /Game/Athena/Items/Gameplay/Second_Chance/Materials/M_SecondChance_Card_Pickup
graph {
 properties {bUsedWithMeshParticles=true, bUsedWithNiagaraMeshParticles=true, pixelDepthOffsetMode=PDOM_Legacy, usedShadingModels=Default Lit}
 nodes [
  n1 = Constant() {r=30}
  LinearInterpolate(
   A=Multiply(A=@n1) {constB=-1}
   B=@n1
   Alpha=MaterialFunctionCall() {materialFunction=&DebugTimeSine}.Result
  )
  n2 = TextureSampleParameter2D() {
   desc="R = SpecMask G = Metallic B = Roughness"
   group=Specular
   parameterName=SpecularMasks
   samplerSource=SSM_Wrap_WorldGroupSettings
   samplerType=SAMPLERTYPE_Masks
   texture=/Game/Athena/Items/Gameplay/Second_Chance/Textures/SecondChance_Card_Pickup_S
  }
  LinearInterpolate(
   A=ScalarParameter() {group=Roughness, parameterName=SpecRoughnessMin}
   B=ScalarParameter() {defaultValue=1, group=Roughness, parameterName=SpecRoughnessMax}
   Alpha=@n2.B
  )
  n3 = ParticlePositionWS()
  n4 = DynamicParameter() {defaultValue=<rgba 0 -100 1 1>, paramNames [WPO,WipePosition,Param3,Param4]}
  n5 = SphereMask(
   A=ComponentMask(None=WorldPosition().XYZ) {b=true}
   B=ComponentMask(None=Add(A=@n3, B=@n4.WipePosition)) {b=true}
   Radius=ScalarParameter() {defaultValue=5, parameterName=WipeSize}
  ) {hardnessPercent=0}
  n6 = Fresnel() {baseReflectFraction=0.1, exponent=2}
  n7 = Constant()
  n8 = ShadingPathSwitch(Default=Multiply(A=Multiply(A=@n5, B=@n6)) {constB=4}, Mobile=@n7) {desc=Migrated from FeatureLevelSwitch}
  n9 = TextureSample() {texture=/Game/Athena/Items/Gameplay/Second_Chance/Textures/SecondChance_Card_Pickup_E}
  n10 = Time()
  n11 = ScalarParameter() {parameterName=StartTime}
  n12 = Clamp(None=Divide(A=Subtract(A=@n10, B=@n11), B=ScalarParameter() {defaultValue=1, parameterName=Duration}))
  n13 = LinearInterpolate(Alpha=@n12) {constA=1, constB=0}
  n14 = Add(A=Clamp(None=LinearInterpolate(Alpha=@n13) {constA=1, constB=-2.5})) {constB=0.25}
  n15 = OneMinus(None=ConstantBiasScale(None=Sine(None=@n14) {period=1.25}))
  n16 = Fresnel(ExponentIn=Multiply(A=@n12) {constB=2}) {baseReflectFraction=0.1}
  n17 = Power(Base=@n13) {constExponent=5}
  n18 = Add(
   A=Multiply(A=@n9.RGB, B=Multiply(A=@n15) {constA=1, constB=5})
   B=Multiply(A=Multiply(A=@n16, B=@n17), B=VectorParameter() {parameterName=GlowColor, defaultValue=<rgba 0.9044 7.6019 8 0>}.RGB)
  )
  MaterialFunctionCall(Emissive=Add(A=StaticSwitchParameter(True=@n8, False=@n7) {parameterName=MeshParticle}, B=QualitySwitch(Default=@n18, Low=@n9.RGB))) {materialFunction=&MF_FX_EmissiveAdjustments}
  Reroute(None=@n2.R)
  n19 = TextureSample() {samplerType=SAMPLERTYPE_Normal, texture=/Game/Athena/Items/Gameplay/Second_Chance/Textures/SecondChance_Card_Pickup_N}
  n20 = Normalize(VectorInput=VectorParameter() {parameterName=RestingRotation, defaultValue=<rgba 0.3 0.2 0 1>}.RGB)
  n21 = Multiply(A=ScalarParameter() {defaultValue=0.08, parameterName=WPORotationTilt}, B=Power(Base=@n12))
  StaticSwitchParameter(
   True=@n19.RGB
   False=MaterialFunctionCall(Rotation Axis=@n20, Rotation Angle=@n21, World Space Vertex Normals=Transform(None=@n19.RGB)) {materialFunction=&FixRotateAboutAxisNormals}.New Tangent Space Vertex Normal
  ) {parameterName=MeshParticle}
  n22 = WorldPosition()
  n23 = ObjectPositionWS()
  n24 = DynamicParameter() {defaultValue=<rgba 0 -100 1 1>, paramNames [WPO,WipePosition,Param3,Param4]}
  n25 = Multiply(A=ComponentMask(None=Subtract(A=@n22.XYZ, B=@n23)) {g=true, r=true}, B=Power(Base=@n24.WPO) {constExponent=5})
  n26 = Power(Base=@n24.WPO)
  n27 = Clamp(None=LinearInterpolate(Alpha=@n24.WPO) {constA=1, constB=0})
  n28 = Add(A=Multiply(A=@n27, B=LinearInterpolate(Alpha=@n27) {constA=1, constB=3})) {constB=0.25}
  n29 = Multiply(A=@n26, B=Multiply(A=@n26, B=Sine(None=@n28) {period=0.75}))
  n30 = RotateAboutAxis(
   NormalizedRotationAxis=@n20
   RotationAngle=@n21
   PivotPoint=MaterialFunctionCall() {materialFunction=&ObjectPivotPoint}.Object Pivot Location
   Position=WorldPosition().XYZ
  )
  n31 = VectorParameter() {parameterName=WPOMax, defaultValue=<rgba -0.7 -0.7 -50 1>}
  n32 = Constant3Vector() {constant=<rgba 0 0 -0.25 0>}
  n33 = ScalarParameter()
  n34 = Subtract(A=WorldPosition().XYZ, B=Add(A=ObjectPositionWS(), B=Multiply(A=@n32, B=@n33) {constA=-0.25}))
  n35 = Power(Base=@n13)
  n36 = Clamp(None=Divide(A=ComponentMask(None=@n34) {b=true}, B=@n33)) {minDefault=0.3}
  n37 = Multiply(A=@n12, B=LinearInterpolate(Alpha=@n12) {constA=1, constB=3})
  n38 = Multiply(A=@n35, B=Sine(None=Add(A=@n37) {constB=0.25}) {period=0.75})
  n39 = AppendVector(A=Multiply(A=ComponentMask(None=@n34) {g=true, r=true}, B=@n17), B=Multiply(A=@n35, B=Multiply(A=@n36, B=@n38)))
  StaticSwitchParameter(
   True=Multiply(A=VectorParameter() {parameterName=WPOMax, defaultValue=<rgba -0.7 -0.7 -50 1>}.RGB, B=AppendVector(A=@n25, B=@n29)) {constB=0}
   False=Add(A=Multiply(A=@n30, B=@n12), B=Multiply(A=@n31.RGB, B=@n39) {constB=0})
  ) {parameterName=MeshParticle}
  TextureSample() {texture=/Game/Athena/Items/Gameplay/Second_Chance/Textures/SecondChance_Card_Pickup_D}
  Time() {bOverride_Period=true, period=1}
 ]
}
