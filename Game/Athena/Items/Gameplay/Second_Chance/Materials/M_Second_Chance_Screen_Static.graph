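# Lit material for the Second Chance screen that shows animated TV static with scanlines.
# Two static samples panning at different rates add to a clamped, desaturated sine flicker and panning scanlines.
# A soft sphere mask and inverted Fresnel dim the edges, and the global time-of-day parameter blends in a night brightness scale.
# Low quality switches drop the scanlines and edge dimming.
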Material /Game/Athena/Items/Gameplay/Second_Chance/Materials/M_Second_Chance_Screen_Static
graph {
 properties {
  bUsedWithMeshDeformer=true
  bUsedWithNanite=true
  bUsedWithSkeletalMesh=true
  pixelDepthOffsetMode=PDOM_Legacy
  usedShadingModels=Default Lit
 }
 nodes [
  n1 = Time() {bOverride_Period=true, period=15}
  n2 = AppendVector(A=AppendVector(A=Sine(None=@n1) {period=3}, B=Sine(None=@n1) {period=1.5}), B=Sine(None=@n1) {period=1.2})
  Ceil(None=@n2)
  Constant() {r=0.5}
  Constant() {r=0.1}
  n3 = Constant() {r=1}
  n4 = ScalarParameter() {defaultValue=1, parameterName=HeightScale}
  n5 = Panner(Coordinate=Multiply(
    A=TextureCoordinate() {vTiling=8}
    B=MaterialFunctionCall(X=@n3, Y=@n4) {materialFunction=&MakeFloat2}.Result
   )) {speedX=0.5, speedY=0.7}
  n6 = QualitySwitch(
   Default=Multiply(A=TextureSample(UVs=@n5) {texture=/Game/Missions/Primary/LaunchTheBalloon/Textures/T_TV_Scanlines}.RGB) {constB=0.2}
   Low=Constant()
  )
  n7 = TextureCoordinate() {uTiling=0.5, vTiling=0.5}
  n8 = Panner(Coordinate=@n7) {speedX=1, speedY=5}
  n9 = Panner(Coordinate=@n7) {speedX=4, speedY=0.5}
  n10 = Add(
   A=Desaturation(None=Clamp(None=@n2) {maxDefault=0.25, minDefault=0.15})
   B=Add(
    A=TextureSample(UVs=@n8) {texture=/Game/Athena/Items/Gameplay/Second_Chance/Textures/T_Van_TV_Static_01_D}.RGB
    B=TextureSample(UVs=@n9) {texture=&T_Van_TV_Static_01_D}.RGB
   )
  )
  n11 = ScalarParameter() {defaultValue=0.5, parameterName=FrasnelDarkenExp}
  n12 = Multiply(
   A=SphereMask(A=TextureCoordinate(), B=Constant() {r=0.5}, Radius=ScalarParameter() {defaultValue=0.74, parameterName=EdgeDimRadius}) {hardnessPercent=0}
   B=OneMinus(None=Fresnel(ExponentIn=@n11) {baseReflectFraction=0, exponent=0.5})
  )
  n13 = Constant() {r=1}
  n14 = CollectionParameter() {
   collection=&FortniteMaterialParameters
   parameterName=FortniteActiveTimeOfDay
  }
  Multiply(
   A=Multiply(A=Add(A=@n6, B=@n10), B=QualitySwitch(Default=@n12, Low=@n13))
   B=LinearInterpolate(B=ScalarParameter() {defaultValue=0.55, parameterName=NightBrightness}, Alpha=ComponentMask(None=@n14) {a=true}) {constA=1}
  )
  Multiply(
   A=@n7
   B=MaterialFunctionCall(X=Constant() {r=1}, Y=ScalarParameter() {defaultValue=1, parameterName=HeightScale}) {materialFunction=&MakeFloat2}.Result
  )
 ]
}
